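# Sound cue for an occasional long seagull ambience spot in the Beach Resort POI, using a looping Pirate Cove seagull recording.
# A Random node picks one of two Envelopers over the shared looping wave, each fading in and out with a different volume contour to shape a finite-sounding pass.
# A Mixer boosts the result, which plays in the outdoor ambient sound class.
# Overrides set distance attenuation without reverb send and allow one instance, blocking new ones.
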SoundCue /Game/Sounds/Asteria/POI/Beach_Resort/Spots/Spot_Bird_Seagull_Long_Rare_Oneshot_Cue
root=Mixer_0
nodes [
 {
  id=Enveloper_0
  children [WavePlayer_0]
  properties {
   loopEnd=70
   pitchCurve {editorCurveData {keys [{time=0}]}}
   volumeCurve {editorCurveData {keys [<at 0 0>,<at 2 1>,<at 15 1>,<at 25 0.2>,<at 40 0>,<at 50 0>]}}
  }
 }
 {
  id=Enveloper_1
  children [WavePlayer_0]
  properties {
   loopEnd=70
   pitchCurve {editorCurveData {keys [{time=0}]}}
   volumeCurve {editorCurveData {keys [<at 0 0>,<at 3 0.5>,<at 10.8264 0.4>,<at 25 1>,<at 35 1>,<at 50 0>]}}
  }
 }
 {id=Mixer_0, children [Random_0], properties {inputVolume [2.5]}}
 {id=Random_0, children [Enveloper_0,Enveloper_1], properties {weights [1,1]}}
 {id=WavePlayer_0, properties {bLooping=true, soundWaveAssetPtr=/Game/Athena/Sounds/Ambient/PirateCove/Amb_PirateCove_Seagull_Loop_01}}
]
properties {
 bOverrideAttenuation=true
 bOverrideConcurrency=true
 maxDistance=12000
 soundClassObject=&Fort_Ambient_Outdoors_Athena
 volumeMultiplier=0.601
 attenuationOverrides {
  bEnableReverbSend=false
  falloffDistance=12000
  manualReverbSendLevel=25
  nonSpatializedRadiusEnd=3500
  reverbSendMethod=Manual
  stereoSpread=2048
  attenuationShapeExtents {x=0}
 }
 concurrencyOverrides {maxCount=1, resolutionRule=PreventNew}
}
