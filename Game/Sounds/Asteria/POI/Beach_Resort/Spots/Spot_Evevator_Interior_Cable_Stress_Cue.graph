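# Sound cue for one-shot elevator cable stress sounds, a spot ambience for an elevator interior at the Asteria Beach Resort POI.
# A Random node picks among nine equally weighted elevator shaft cable foley waves.
# A Modulator at the root varies volume and lowers pitch by a random amount on each play.
# The cue uses the ambient one-shot sound class and a tile ambient one-shot attenuation asset, with reduced overall volume.
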SoundCue /Game/Sounds/Asteria/POI/Beach_Resort/Spots/Spot_Evevator_Interior_Cable_Stress_Cue
root=Modulator_0
nodes [
 {id=Modulator_0, children [Random_0], properties {pitchMax=1, pitchMin=0.8, volumeMax=1, volumeMin=0.7}}
 {
  id=Random_0
  children [WavePlayer_0,WavePlayer_1,WavePlayer_2,WavePlayer_3,WavePlayer_4,WavePlayer_5,WavePlayer_6,WavePlayer_7,WavePlayer_8]
  properties {weights [1,1,1,1,1,1,1,1,1]}
 }
 {id=WavePlayer_0, properties {soundWaveAssetPtr=&/Spot_Evelator_Shaft_Cable_Foley_01}}
 {id=WavePlayer_1, properties {soundWaveAssetPtr=&/Spot_Evelator_Shaft_Cable_Foley_02}}
 {id=WavePlayer_2, properties {soundWaveAssetPtr=&/Spot_Evelator_Shaft_Cable_Foley_03}}
 {id=WavePlayer_3, properties {soundWaveAssetPtr=&/Spot_Evelator_Shaft_Cable_Foley_04}}
 {id=WavePlayer_4, properties {soundWaveAssetPtr=&/Spot_Evelator_Shaft_Cable_Foley_05}}
 {id=WavePlayer_5, properties {soundWaveAssetPtr=&/Spot_Evelator_Shaft_Cable_Foley_06}}
 {id=WavePlayer_6, properties {soundWaveAssetPtr=&/Spot_Evelator_Shaft_Cable_Foley_07}}
 {id=WavePlayer_7, properties {soundWaveAssetPtr=&/Spot_Evelator_Shaft_Cable_Foley_08}}
 {id=WavePlayer_8, properties {soundWaveAssetPtr=&/Spot_Evelator_Shaft_Cable_Foley_09}}
]
properties {
 attenuationSettings=/Game/Sounds/Attenuation/Fort_TileAmbient_OneShot_Att
 maxDistance=8000
 soundClassObject=&Fort_Ambient_OneShots
 volumeMultiplier=0.451
}
