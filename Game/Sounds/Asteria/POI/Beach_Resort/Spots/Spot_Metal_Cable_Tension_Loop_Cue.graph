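# Looping ambient spot cue for creaking metal cable under tension at the Beach Resort POI, reusing nine creak waves from the Blimp Crashsite set.
# Each loop waits a random delay, then plays a random creak with slight upward pitch variation.
# A random branch plays it plain, with a late fade-out envelope or with a late fade-in envelope, identically across quality levels.
# Overridden attenuation with occlusion and a tight concurrency limit shape playback.
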SoundCue /Game/Sounds/Asteria/POI/Beach_Resort/Spots/Spot_Metal_Cable_Tension_Loop_Cue
root=Looping_0
nodes [
 {id=Delay_1, children [QualityLevel_1], properties {delayMax=2, delayMin=0.5}}
 {
  id=Enveloper_0
  children [Modulator_0]
  properties {
   pitchCurve {editorCurveData {keys [{time=0}]}}
   volumeCurve {editorCurveData {keys [<at 0 1>, {arriveTangent=-0.2, interpMode=RCIM_Cubic, leaveTangent=-0.2, time=3, value=1}, <at 5 0>]}}
  }
 }
 {
  id=Enveloper_1
  children [Modulator_0]
  properties {
   pitchCurve {editorCurveData {keys [{time=0}]}}
   volumeCurve {editorCurveData {keys [<at 0 0 RCIM_Cubic>, {arriveTangent=0.0151, interpMode=RCIM_Cubic, leaveTangent=0.0151, tangentMode=RCTM_User, time=3, value=0}, <at 5 1>]}}
  }
 }
 {id=Looping_0, children [Delay_1]}
 {id=Modulator_0, children [Random_0], properties {pitchMax=1.2, pitchMin=1}}
 {id=QualityLevel_1, children [Random_1,Random_1,Random_1,Random_1]}
 {
  id=Random_0
  children [WavePlayer_0,WavePlayer_1,WavePlayer_2,WavePlayer_3,WavePlayer_4,WavePlayer_5,WavePlayer_6,WavePlayer_7,WavePlayer_8]
  properties {weights [1,1,1,1,1,1,1,1,1]}
 }
 {id=Random_1, children [Enveloper_0,Modulator_0,Enveloper_1], properties {weights [1,1,1]}}
 {id=WavePlayer_0, properties {soundWaveAssetPtr=/Game/Sounds/Artemis/POI/Blimps_Crashsites/Spots/Spot_Blimp_Crashsite_Cable_Tension_Creak_09}}
 {id=WavePlayer_1, properties {soundWaveAssetPtr=/Game/Sounds/Artemis/POI/Blimps_Crashsites/Spots/Spot_Blimp_Crashsite_Cable_Tension_Creak_01}}
 {id=WavePlayer_2, properties {soundWaveAssetPtr=/Game/Sounds/Artemis/POI/Blimps_Crashsites/Spots/Spot_Blimp_Crashsite_Cable_Tension_Creak_02}}
 {id=WavePlayer_3, properties {soundWaveAssetPtr=/Game/Sounds/Artemis/POI/Blimps_Crashsites/Spots/Spot_Blimp_Crashsite_Cable_Tension_Creak_03}}
 {id=WavePlayer_4, properties {soundWaveAssetPtr=/Game/Sounds/Artemis/POI/Blimps_Crashsites/Spots/Spot_Blimp_Crashsite_Cable_Tension_Creak_04}}
 {id=WavePlayer_5, properties {soundWaveAssetPtr=/Game/Sounds/Artemis/POI/Blimps_Crashsites/Spots/Spot_Blimp_Crashsite_Cable_Tension_Creak_05}}
 {id=WavePlayer_6, properties {soundWaveAssetPtr=/Game/Sounds/Artemis/POI/Blimps_Crashsites/Spots/Spot_Blimp_Crashsite_Cable_Tension_Creak_06}}
 {id=WavePlayer_7, properties {soundWaveAssetPtr=/Game/Sounds/Artemis/POI/Blimps_Crashsites/Spots/Spot_Blimp_Crashsite_Cable_Tension_Creak_07}}
 {id=WavePlayer_8, properties {soundWaveAssetPtr=/Game/Sounds/Artemis/POI/Blimps_Crashsites/Spots/Spot_Blimp_Crashsite_Cable_Tension_Creak_08}}
]
properties {
 attenuationSettings=/Game/Sounds/Artemis/POI/Blimps_Crashsites/Spots/Spot_Wind_Metal_Cable_Creaks_Box_Att
 bEnableBusSends=false
 bOverrideAttenuation=true
 bOverrideConcurrency=true
 maxDistance=8500
 soundClassObject=&Fort_Gadgets
 volumeMultiplier=0.5
 attenuationOverrides {
  falloffDistance=6000
  occlusionLowPassFilterFrequency=2000
  occlusionTraceChannel=ECC_GameTraceChannel11
  occlusionVolumeAttenuation=0.5
  stereoSpread=50
  attenuationShapeExtents=<2500 85>
 }
 busSends [{audioBus=null, minSendLevel=0, sendLevel=1, soundSourceBus=null}]
 concurrencyOverrides {maxCount=2, voiceStealReleaseTime=0.2}
 soundSubmixSends [{minSendLevel=0.6, soundSubmix=&Fortnite_Reverb_Default}]
}
