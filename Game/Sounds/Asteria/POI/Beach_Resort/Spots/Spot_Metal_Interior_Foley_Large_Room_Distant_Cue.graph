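# Sound cue for a distant metal-interior shuffling foley spot ambience in the Asteria Beach Resort POI, named for large rooms.
# A Random node picks among nine equally weighted distant shuffling-movement waves.
# A Modulator above it varies each play with slightly lowered random pitch and volume.
# It plays at about half volume in the ambient one-shot sound class, using the tile ambient one-shot attenuation.
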SoundCue /Game/Sounds/Asteria/POI/Beach_Resort/Spots/Spot_Metal_Interior_Foley_Large_Room_Distant_Cue
root=Modulator_0
nodes [
 {id=Modulator_0, children [Random_1], properties {pitchMax=1, pitchMin=0.8, volumeMax=1, volumeMin=0.7}}
 {
  id=Random_1
  children [WavePlayer_30, WavePlayer_31, WavePlayer_32, WavePlayer_33, WavePlayer_34, WavePlayer_35, WavePlayer_36, WavePlayer_37, WavePlayer_38]
  properties {weights [1,1,1,1,1,1,1,1,1]}
 }
 {
  id=WavePlayer_30
  properties {soundWaveAssetPtr=&/Spot_Metal_Interior_Shuffling_Movement_Foley_Distant_01}
 }
 {
  id=WavePlayer_31
  properties {soundWaveAssetPtr=&/Spot_Metal_Interior_Shuffling_Movement_Foley_Distant_03}
 }
 {
  id=WavePlayer_32
  properties {soundWaveAssetPtr=&/Spot_Metal_Interior_Shuffling_Movement_Foley_Distant_09}
 }
 {
  id=WavePlayer_33
  properties {soundWaveAssetPtr=&/Spot_Metal_Interior_Shuffling_Movement_Foley_Distant_10}
 }
 {
  id=WavePlayer_34
  properties {soundWaveAssetPtr=&/Spot_Metal_Interior_Shuffling_Movement_Foley_Distant_11}
 }
 {
  id=WavePlayer_35
  properties {soundWaveAssetPtr=&/Spot_Metal_Interior_Shuffling_Movement_Foley_Distant_16}
 }
 {
  id=WavePlayer_36
  properties {soundWaveAssetPtr=&/Spot_Metal_Interior_Shuffling_Movement_Foley_Distant_17}
 }
 {
  id=WavePlayer_37
  properties {soundWaveAssetPtr=&/Spot_Metal_Interior_Shuffling_Movement_Foley_Distant_18}
 }
 {
  id=WavePlayer_38
  properties {soundWaveAssetPtr=&/Spot_Metal_Interior_Shuffling_Movement_Foley_Distant_19}
 }
]
properties {
 attenuationSettings=/Game/Sounds/Attenuation/Fort_TileAmbient_OneShot_Att
 maxDistance=8000
 soundClassObject=&Fort_Ambient_OneShots
 volumeMultiplier=0.501
}
