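# Looping ambient spot cue named for the Mountain Hotel hologram at Asteria's Beach Resort POI, though it has no root or sound nodes, so it plays no audio as authored here.
# The remaining settings route it to the Athena ambient sound class at reduced volume, with a two-voice concurrency limit.
# Attenuation overrides define a short falloff and configure interpolated focus, with quieter non-focus and occluded levels.
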SoundCue /Game/Sounds/Asteria/POI/Beach_Resort/Spots/Spot_Mountain_Hotel_Hologram_Loop_Cue
root=null
nodes []
properties {
 bOverrideConcurrency=true
 maxDistance=2097152
 soundClassObject=&Fort_Ambient_Athena
 volumeMultiplier=0.401
 attenuationOverrides {
  bEnableFocusInterpolation=true
  falloffDistance=600
  nonFocusVolumeAttenuation=0.3
  occlusionVolumeAttenuation=0.3
  stereoSpread=0
  attenuationShapeExtents {x=200}
 }
 concurrencyOverrides {maxCount=2}
}
