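# Sound cue for a toilet drip ambient spot in the Beach Resort POI, routed to the indoor ambient one-shot sound class.
# A Random node picks among six equally weighted stereo drip waves from the Hollow Base folder, feeding a Modulator that varies pitch slightly upward.
# It plays at reduced volume with positional one-shot attenuation, a two-voice concurrency override, and bus and submix sends disabled.
# Despite the Loop name, the cue has no looping node.
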SoundCue /Game/Sounds/Asteria/POI/Beach_Resort/Spots/Spot_Toilet_Drip_Single_Smaller_Loop_Cue
root=Modulator_0
nodes [
 {id=Modulator_0, children [Random_0], properties {pitchMax=1.1, pitchMin=1}}
 {
  id=Random_0
  children [WavePlayer_18, WavePlayer_25, WavePlayer_26, WavePlayer_27, WavePlayer_28, WavePlayer_29]
  properties {weights [1,1,1,1,1,1]}
 }
 {id=WavePlayer_18, properties {soundWaveAssetPtr=/Game/Sounds/Artemis/POI/Hollow_Base/Spots/Spot_Toilet_Drips_13_Stereo}}
 {id=WavePlayer_25, properties {soundWaveAssetPtr=/Game/Sounds/Artemis/POI/Hollow_Base/Spots/Spot_Toilet_Drips_14_Stereo}}
 {id=WavePlayer_26, properties {soundWaveAssetPtr=/Game/Sounds/Artemis/POI/Hollow_Base/Spots/Spot_Toilet_Drips_15_Stereo}}
 {id=WavePlayer_27, properties {soundWaveAssetPtr=/Game/Sounds/Artemis/POI/Hollow_Base/Spots/Spot_Toilet_Drips_16_Stereo}}
 {id=WavePlayer_28, properties {soundWaveAssetPtr=/Game/Sounds/Artemis/POI/Hollow_Base/Spots/Spot_Toilet_Drips_17_Stereo}}
 {id=WavePlayer_29, properties {soundWaveAssetPtr=/Game/Sounds/Artemis/POI/Hollow_Base/Spots/Spot_Toilet_Drips_18_Stereo}}
]
properties {
 attenuationSettings=/Game/Sounds/Attenuation/Fort_TileAmbient_OneShot_MorePositional_Att
 bEnableBusSends=false
 bEnableSubmixSends=false
 bOverrideConcurrency=true
 maxDistance=8000
 soundClassObject=/Game/Sounds/FortSoundClassesAndModes/Fort_Ambient_OneShots_Indoors
 volumeMultiplier=0.251
 attenuationOverrides {falloffDistance=6000, stereoSpread=10000, attenuationShapeExtents {x=2000}}
 concurrencyOverrides {maxCount=2, voiceStealReleaseTime=2}
}
