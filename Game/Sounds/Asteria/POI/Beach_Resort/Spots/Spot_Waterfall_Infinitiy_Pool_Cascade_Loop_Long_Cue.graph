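# Looping ambient Sound Cue for the infinity pool waterfall cascade at the Asteria Beach Resort POI, routed to the water ambient sound class.
# A mixer layers close and distant loops, each with its own box attenuation and listener focus: the close layer adds occlusion, the distant one reaches farther with distance low-pass filtering.
# The mix passes through slight pitch modulation, a volume-dip envelope and a short fixed delay, with a tight concurrency limit.
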SoundCue /Game/Sounds/Asteria/POI/Beach_Resort/Spots/Spot_Waterfall_Infinitiy_Pool_Cascade_Loop_Long_Cue
root=Delay_0
nodes [
 {
  id=Attenuation_2
  children [WavePlayer_0]
  properties {
   bOverrideAttenuation=true
   attenuationOverrides {
    attenuationShape=Box
    bEnableListenerFocus=true
    bEnableOcclusion=true
    falloffDistance=750
    nonFocusVolumeAttenuation=0.5
    nonSpatializedRadiusEnd=100
    occlusionInterpolationTime=5
    occlusionTraceChannel=ECC_GameTraceChannel11
    occlusionVolumeAttenuation=0.2
    attenuationShapeExtents=<721 70 263>
   }
  }
 }
 {
  id=Attenuation_3
  children [WavePlayer_2]
  properties {
   bOverrideAttenuation=true
   attenuationOverrides {
    attenuationShape=Box
    bAttenuateWithLPF=true
    bEnableListenerFocus=true
    distanceAlgorithm=Logarithmic
    falloffDistance=2000
    lPFFrequencyAtMax=5000
    lPFFrequencyAtMin=22000
    lPFRadiusMax=3000
    lPFRadiusMin=750
    nonFocusVolumeAttenuation=0.5
    nonSpatializedRadiusEnd=100
    occlusionInterpolationTime=5
    occlusionTraceChannel=ECC_GameTraceChannel11
    occlusionVolumeAttenuation=0.2
    attenuationShapeExtents=<721 350 263>
   }
  }
 }
 {id=Delay_0, children [Enveloper_0], properties {delayMax=0.2, delayMin=0.2}}
 {
  id=Enveloper_0
  children [Modulator_0]
  properties {pitchCurve {editorCurveData {keys [{time=0}]}}, volumeCurve {editorCurveData {keys [<at 0 1 RCIM_Cubic>,<at 0.5 0.8 RCIM_Cubic>,<at 1 1>]}}}
 }
 {id=Mixer_0, children [Attenuation_2, Attenuation_3], properties {inputVolume [0.9,0.9]}}
 {id=Modulator_0, children [Mixer_0], properties {pitchMax=1.1, pitchMin=1}}
 {
  id=WavePlayer_0
  properties {bLooping=true, soundWaveAssetPtr=&/Spot_Waterfall_Infinitiy_Pool_Cascade_Close}
 }
 {
  id=WavePlayer_2
  properties {bLooping=true, soundWaveAssetPtr=&/Spot_Waterfall_Infinitiy_Pool_Cascade_Dist}
 }
]
properties {
 bEnableBusSends=false
 bEnableSubmixSends=false
 bOverrideConcurrency=true
 maxDistance=2721
 soundClassObject=/Game/Sounds/FortSoundClassesAndModes/Fort_Water_Ambient
 volumeMultiplier=0.5
 attenuationOverrides {
  attenuationShape=Capsule
  bAttenuate=false
  bEnableOcclusion=true
  distanceAlgorithm=Logarithmic
  falloffDistance=2000
  occlusionInterpolationTime=5
  occlusionLowPassFilterFrequency=1000
  occlusionTraceChannel=ECC_GameTraceChannel11
  occlusionVolumeAttenuation=0.2
  stereoSpread=10000
  attenuationShapeExtents=<1000 60>
 }
 concurrencyOverrides {maxCount=2, voiceStealReleaseTime=2}
}
